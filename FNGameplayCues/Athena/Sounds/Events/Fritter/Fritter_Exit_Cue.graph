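# Sound cue for a player exiting the water wall in the Fritter event.
# A single wave player feeds the root with the exit sound, with no randomization or mixing.
# It plays at reduced volume in the scripted events sound class, using grappling gadget attenuation.
# Overridden concurrency allows one instance per owner and blocks new plays while one is active.
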SoundCue /FNGameplayCues/Athena/Sounds/Events/Fritter/Fritter_Exit_Cue
root=WavePlayer_0
nodes [{id=WavePlayer_0, properties {soundWaveAssetPtr=/FNGameplayCues/Athena/Sounds/Events/Fritter/Wavs/Event_Fritter_SFX_Water_Wall_Player_Exit}}]
properties {
 attenuationSettings=/Game/Sounds/Attenuation/Fort_GadgetGrappling_Att
 bOverrideConcurrency=true
 maxDistance=8500
 soundClassObject=/Game/Sounds/FortSoundClassesAndModes/Fort_Events_Scripted
 volumeMultiplier=0.3144
 concurrencyOverrides {bLimitToOwner=true, maxCount=1, resolutionRule=PreventNew}
}
